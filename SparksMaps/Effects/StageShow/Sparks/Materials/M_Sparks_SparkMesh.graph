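# Unlit masked material for Niagara stage-show spark meshes, animating glowing orange sparks in a looping burst.
# Time plus vertex-color red drives the loop; green and blue vary motion, scaled by Wobble, Spread and Height dynamic parameters.
# The Sprite function and camera-based rotation billboard the geometry.
# A texture-free SphereMask draws each streak, particle alpha controls fading, and an HDR color blend feeds an emissive adjustment function.
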Material /SparksMaps/Effects/StageShow/Sparks/Materials/M_Sparks_SparkMesh
graph {
 properties {
  bEnableResponsiveAA=true
  bHasPixelAnimation=true
  bUsedWithNiagaraMeshParticles=true
  bUsedWithNiagaraSprites=true
  blendMode=BLEND_Masked
  pixelDepthOffsetMode=PDOM_Legacy
  shadingModel=MSM_Unlit
  usedShadingModels=Unlit
 }
 nodes [
  n1 = Constant3Vector() {constant=<rgba 10 3.417 1 1>}
  n2 = Constant3Vector() {constant=<rgba 10 1.75 1 1>}
  n3 = Power(Base=NamedRerouteUsage()) {constExponent=0.5}
  MaterialFunctionCall(Emissive=Multiply(A=LinearInterpolate(A=@n1, B=@n2, Alpha=@n3), B=NamedRerouteUsage())) {materialFunction=&MF_FX_EmissiveAdjustments}
  n4 = Power(Base=NamedRerouteUsage()) {constExponent=0.5}
  n5 = Saturate(None=Multiply(A=@n4, B=Constant() {r=2}) {constB=2})
  n6 = OneMinus(None=@n4)
  MaterialFunctionCall!(
   Center Location=WorldPosition().XYZ
   0-1 UVs
   XY scale=AppendVector(A=Multiply(B=@n5) {constA=15}, B=Multiply(B=@n6) {constA=50})
   OverridePivot
   "Normals (Optional)"
  ) {materialFunction=&Sprite}
  n7 = ParticleColor()
  n8 = Subtract(A=Add(A=NamedRerouteUsage(), B=@n7.A), B=OneMinus(None=@n7.A))
  n9 = Power(Base=NamedRerouteUsage()) {constExponent=4}
  n10 = Multiply(A=TextureCoordinate(), B=AppendVector(A=LinearInterpolate(Alpha=@n9) {constA=1, constB=10}, B=Constant() {r=1}))
  n11 = MaterialFunctionCall(Float2=Abs(None=ConstantBiasScale(None=@n10) {bias=-0.5, scale=2})) {materialFunction=&BreakOutFloat2Components}
  n12 = SphereMask(
   A=MaterialFunctionCall(X=@n11.R, Y=Power(Base=@n11.G) {constExponent=3}) {materialFunction=&MakeFloat2}.Result
   B=Constant()
   Radius=Constant() {r=1}
   Hardness=Constant()
  )
  n13 = Constant() {r=2}
  Multiply(
   A=NamedRerouteDeclaration(None=Power(Base=@n12, Exp=@n13) {constExponent=3}) {name=Spark Texture, nodeColor=<rgba 0 0 0 1>}
   B=Saturate(None=Divide(A=@n8) {constB=0.1})
  )
  n14 = Multiply(A=Time(), B=ScalarParameter() {defaultValue=1.5, group=Movement, parameterName=Speed})
  n15 = VertexColor()
  NamedRerouteDeclaration(None=Frac(None=Add(A=@n14, B=@n15.R))) {name=LoopingAnim, nodeColor=<rgba 0 0 0 1>}
  n16 = Subtract(
   A=Transform(None=WorldPosition().XYZ) {transformSourceType=TRANSFORMSOURCE_Local}
   B=Transform(None=ParticlePositionWS()) {transformSourceType=TRANSFORMSOURCE_Local}
  )
  NamedRerouteDeclaration(None=Divide(A=ComponentMask(None=@n16) {b=true}) {constB=500}) {name=Object Gradient, nodeColor=<rgba 0 0 0 1>}
  n17 = CameraPositionWS()
  n18 = Constant3Vector() {constant=<rgba 1 1 0 0>}
  n19 = Normalize(VectorInput=Subtract(A=ObjectPositionWS(), B=Multiply(A=@n17, B=@n18)))
  n20 = Arctangent2Fast(Y=ComponentMask(None=@n19) {g=true}, X=ComponentMask(None=@n19) {r=true})
  n21 = Add(B=Divide(A=Divide(A=@n20) {constB=3.1415})) {constA=0.25, constB=0.25}
  NamedRerouteDeclaration(None=RotateAboutAxis(
    NormalizedRotationAxis=ObjectOrientation()
    RotationAngle=@n21
    PivotPoint=MaterialFunctionCall() {materialFunction=&ObjectPivotPoint}.Object Pivot Location
    Position=WorldPosition().XYZ
   )) {name=Billboard, nodeColor=<rgba 0 0 0 1>}
  n22 = VertexColor()
  n23 = DynamicParameter() {defaultValue=<rgba 50 1 500 1>, paramNames [Wobble,Spread,Height,Param4]}
  n24 = Multiply(
   A=AppendVector(A=LinearInterpolate(Alpha=@n22.G) {constA=-75, constB=75}, B=LinearInterpolate(Alpha=@n22.B) {constA=-75, constB=75})
   B=@n23.Spread
  ) {constB=2}
  n25 = Constant() {r=0.65}
  n26 = Clamp(None=VertexColor().G) {minDefault=0.75}
  n27 = Multiply(A=AppendVector(A=Multiply(A=@n24, B=@n25), B=Multiply(A=@n23.Height, B=@n26)), B=NamedRerouteUsage())
  n28 = Multiply(A=Time())
  n29 = VertexColor()
  n30 = Add(A=@n28, B=@n29.G)
  n31 = NamedRerouteUsage()
  n32 = Add(A=@n28, B=@n29.G)
  MaterialExpressionAppend3Vector_0 = MaterialXAppend3Vector!(A=Multiply(A=Sine(None=@n30), B=@n31), B=Multiply(A=Cosine(None=@n32), B=@n31), C=Constant())
  n33 = Multiply(A=@MaterialExpressionAppend3Vector_0, B=DynamicParameter() {defaultValue=<rgba 50 1 500 1>, paramNames [Wobble,Spread,Height,Param4]}.Wobble) {constB=50}
  Reroute(None=Add(A=NamedRerouteUsage(), B=Add(A=@n27, B=@n33)))
  Subtract(A=@n21, B=MaterialFunctionCall() {materialFunction=/Engine/ArtTools/RenderToTexture/MaterialFunctions/FoliageZRotation}."Z Rotation (0-1)")
 ]
}
